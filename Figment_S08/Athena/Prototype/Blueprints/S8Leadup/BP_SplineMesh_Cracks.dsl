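; Prototype spline mesh actor for Season 8 lead-up ground cracks, with an optional Y-split into two branches.
; Construction builds zigzag segments and socketed branches, giving each dynamic material a sequential ProgressOffset.
; Custom events carve the landscape along splines or trace-snap points onto it, while BeginPlay only enables collision.
; MoveDamage maps MPC CrackProgression along splines to collect nearby building actors, but nothing here calls it and DestroyActors is empty.
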(event Custom|SetDestructionTimers
  (Utilities|Time|SetTimerbyEvent (EventDispatchers|CreateEvent :!event Custom|DestroyActors) 0.001))

(event Custom|DestroyActors)

(event Custom|CleanupExtraComponents)

(event Custom|EnableCollision
  (for _array_element (Variables|Default|GetSplineMeshes)
    (Collision|SetCollisionProfileName _array_element "FortStaticMesh" true :!class /Script/Engine.PrimitiveComponent))
  (Collision|SetCollisionProfileName (Variables|SplineMeshActor|GetSplineMeshComponent) "FortStaticMesh" true :!class /Script/Engine.PrimitiveComponent))

(event Custom|StartCameraShakes)

(event EventBeginPlay
  ;[1 1 1] Camera shakes and rumble
  (CallFunction|EnableCollision))

(event Custom|Conform%20Landscape%20to%20Spline_1
  (CallFunction|ConformLandscapetoSplineComponents (Variables|Default|GetSpline) (Variables|Default|GetConfromPushUpZ) (Variables|Default|GetConformWidth) (Variables|Default|GetConformFalloff))
  (CallFunction|ConformLandscapetoSplineComponents (Variables|Default|GetSpline) (Variables|Default|GetConfromPushDownCenter) (Variables|Default|GetConformWidth_Center) (Variables|Default|GetConformFalloff_Center)))

(event Custom|Conform%20Landscape%20to%20Spline_Left
  (CallFunction|ConformLandscapetoSplineComponents (Variables|Default|GetSplineLeft) (Variables|Default|GetConfromPushUpZ) (Variables|Default|GetConformWidth) (Variables|Default|GetConformFalloff))
  (CallFunction|ConformLandscapetoSplineComponents (Variables|Default|GetSplineLeft) (Variables|Default|GetConfromPushDownCenter) (Variables|Default|GetConformWidth_Center) (Variables|Default|GetConformFalloff_Center)))

(event Custom|Conform%20Landscape%20to%20Spline_Right
  (CallFunction|ConformLandscapetoSplineComponents (Variables|Default|GetSplineRight) (Variables|Default|GetConfromPushUpZ) (Variables|Default|GetConformWidth) (Variables|Default|GetConformFalloff))
  (CallFunction|ConformLandscapetoSplineComponents (Variables|Default|GetSplineRight) (Variables|Default|GetConfromPushDownCenter) (Variables|Default|GetConformWidth_Center) (Variables|Default|GetConformFalloff_Center)))

(event Custom|ConformSpline_1_ToLandscape
  (CallFunction|ConfrormSplinetoLanscape (Variables|Default|GetSpline)))

(event Custom|ConformSpline_Left_ToLandscape
  (CallFunction|ConfrormSplinetoLanscape (Variables|Default|GetSplineLeft)))

(event Custom|ConformSpline_Right_ToLandscape
  (CallFunction|ConfrormSplinetoLanscape (Variables|Default|GetSplineRight)))

(fn AddSplineComponentstoSpline
  (params (Spline :type Spline%20Component%20Object%20Reference) (FistIndex :type Integer) (StartisSocketBase :type Boolean) (EndIsSocketBase :type Boolean) (Material%20List :type Array%20of%20Material%20Instance%20Dynamic%20Object%20References))
  (bind _returnvalue (Spline|GetNumberOfSplinePoints Spline))
  (bind _returnvalue_2 (- _returnvalue 2))
  (bind _zigzagstrenght (Variables|SplineOptions|GetZigZagStrenght))
  (Utilities|FlowControl|Sequence
    (:then_0
      ;[1 1 1] Set align first spline point with Y conncetor
      (if StartisSocketBase
        (Spline|SetLocationatSplinePoint Spline 0 (Math|Vector|MakeVector 0.0) "Local" true)
        (Spline|SetTangentatSplinePoint Spline 0 (Math|Vector|MakeVector 1.0) "Local" true)))
    (:then_1
      (for _index (range FistIndex (+ _returnvalue_2 1))
        (bind _returnvalue_3 (AddComponent|NODE_AddSplineMeshComponent-1 :RelativeTransform (Transformation|GetRelativeTransform Spline)))
        (Utilities|Array|Add (Variables|Default|GetSplineMeshes) _returnvalue_3)
        (Components|StaticMesh|SetStaticMesh _returnvalue_3 (Variables|/Script/Engine.StaticMeshComponent|GetStaticMesh (Variables|SplineMeshActor|GetSplineMeshComponent)))
        (SplineMesh|SetForwardAxis _returnvalue_3 (SplineMesh|GetForwardAxis (Variables|SplineMeshActor|GetSplineMeshComponent)) true)
        (SplineMesh|SetStartandEnd _returnvalue_3 (Spline|GetLocationandTangentatSplinePoint Spline _index "Local") (Math|Vector|SelectVector (* (Math|Vector|Normalize (.Tangent (Spline|GetLocationandTangentatSplinePoint Spline _index "Local"))) (* (Math|Vector|VectorLength (- (Spline|GetLocationandTangentatSplinePoint Spline (+ _index 1) "Local") (Spline|GetLocationandTangentatSplinePoint Spline _index "Local"))) (Variables|SplineOptions|GetTension))) (.Tangent (Spline|GetLocationandTangentatSplinePoint Spline _index "Local")) (Variables|SplineOptions|GetAutoSmoothTangent)) (Spline|GetLocationandTangentatSplinePoint Spline (+ _index 1) "Local") (Math|Vector|SelectVector (* (Math|Vector|Normalize (.Tangent (Spline|GetLocationandTangentatSplinePoint Spline (+ _index 1) "Local"))) (* (Math|Vector|VectorLength (- (Spline|GetLocationandTangentatSplinePoint Spline (+ _index 1) "Local") (Spline|GetLocationandTangentatSplinePoint Spline _index "Local"))) (Variables|SplineOptions|GetTension))) (.Tangent (Spline|GetLocationandTangentatSplinePoint Spline (+ _index 1) "Local")) (Variables|SplineOptions|GetAutoSmoothTangent)) true :!class /Script/Engine.SplineMeshComponent)
        (SplineMesh|SetStartScale _returnvalue_3 (Math|Vector2D|MakeVector2D (.y (Math|Vector|SelectVector (Spline|GetScaleatSplinePoint (Variables|Default|GetSpline) (Spline|GetNumberOfSplinePoints (Variables|Default|GetSpline))) (Spline|GetScaleatSplinePoint Spline _index) (and StartisSocketBase (== _index 0)))) (.z (Math|Vector|SelectVector (Spline|GetScaleatSplinePoint (Variables|Default|GetSpline) (Spline|GetNumberOfSplinePoints (Variables|Default|GetSpline))) (Spline|GetScaleatSplinePoint Spline _index) (and StartisSocketBase (== _index 0))))) true :!class /Script/Engine.SplineMeshComponent)
        (SplineMesh|SetEndScale _returnvalue_3 (Math|Vector2D|MakeVector2D (.y (Spline|GetScaleatSplinePoint Spline (+ _index 1))) (.z (Spline|GetScaleatSplinePoint Spline (+ _index 1)))) true :!class /Script/Engine.SplineMeshComponent)
        (SplineMesh|SetEndOffset _returnvalue_3 (Math|Vector2D|MakeVector2D (Math|Float|SelectFloat 0.0 (Math|Float|SelectFloat (* (Math|Float|%25%28Float%29 (Math|Conversions|ToFloat(Integer) _index) 2.0) _zigzagstrenght) (* _zigzagstrenght -1.0) (> (Math|Float|%25%28Float%29 (Math|Conversions|ToFloat(Integer) _index) 2.0) 0.0)) (and (== _index _returnvalue_2) EndIsSocketBase))) true)
        (SplineMesh|SetStartOffset _returnvalue_3 (Math|Vector2D|MakeVector2D (Math|Float|SelectFloat 0.0 (Math|Float|SelectFloat (* (* (Math|Float|%25%28Float%29 (Math|Conversions|ToFloat(Integer) _index) 2.0) _zigzagstrenght) -1.0) _zigzagstrenght (> (Math|Float|%25%28Float%29 (Math|Conversions|ToFloat(Integer) _index) 2.0) 0.0)) (and (== _index 0) StartisSocketBase))) true)
        (bind _returnvalue_4 (Rendering|Material|CreateDynamicMaterialInstance _returnvalue_3 0 (Variables|SplineOptions|GetMaterialOverride) "None" :!class /Script/Engine.PrimitiveComponent))
        (Utilities|Array|Add Material%20List _returnvalue_4)
        (SplineMesh|SetEndRollRadians _returnvalue_3 (Math|Trig|DegreesToRadians (Spline|GetRollatSplinePoint Spline (+ _index 1) "Local")) true)
        (SplineMesh|SetStartRollRadians _returnvalue_3 (Math|Float|SelectFloat 0.0 (Math|Trig|DegreesToRadians (Spline|GetRollatSplinePoint Spline _index "Local")) (and StartisSocketBase (== _index 0))) true)))))

(fn ConformLandscapeToSplineComponents
  (params (Spline :type Spline%20Component%20Object%20Reference) (OffsetZ :type Float%20%28double-precision%29) (Width :type Float%20%28double-precision%29) (Falloff :type Float%20%28double-precision%29))
  (bind _localsplinesource (LocalVariables|GetLocalSplineSource))
  (bind _autosmoothtangent (Variables|SplineOptions|GetAutoSmoothTangent))
  (bind _tempspline (Variables|Default|GetTempSpline))
  (LocalVariables|SetLocalSplineSource Spline)
  (bind _outactors (Actor|GetAllActorsOfClass "/Script/Landscape.LandscapeProxy"))
  (for _array_element _outactors
    (bind _returnvalue (AddComponent|NODE_AddSplineComponent-3 :RelativeTransform (Math|Transform|MakeTransform (Math|Vector|UnrotateVector (Math|Vector|NegateVector (Transformation|GetActorLocation)) (Transformation|GetActorRotation)) (Math|Rotator|InvertRotator (Transformation|GetActorRotation)) (/ (Math|Vector|MakeVector 1.0 1.0 1.0) (Transformation|GetActorScale3D)))))
    (Variables|Default|SetTempSpline _returnvalue)
    (for _index (range (+ (- (Spline|GetNumberOfSplinePoints (LocalVariables|GetLocalSplineSource)) 2) 1))
      (Spline|SetSplinePoints _tempspline (Utilities|Array|MakeArray (+ (Spline|GetLocationandTangentatSplinePoint _localsplinesource _index "World") (Math|Vector|MakeVector 0.0 0.0 OffsetZ)) (+ (Math|Vector|MakeVector 0.0 0.0 OffsetZ) (Spline|GetLocationandTangentatSplinePoint _localsplinesource (+ _index 1) "World"))) "Local" true)
      (Spline|SetTangentatSplinePoint _tempspline 0 (Math|Vector|SelectVector (* (Math|Vector|Normalize (.Tangent (Spline|GetLocationandTangentatSplinePoint _localsplinesource _index "World"))) (* (Math|Vector|VectorLength (- (Spline|GetLocationandTangentatSplinePoint _localsplinesource (+ _index 1) "World") (Spline|GetLocationandTangentatSplinePoint _localsplinesource _index "World"))) (Variables|SplineOptions|GetTension))) (.Tangent (Spline|GetLocationandTangentatSplinePoint _localsplinesource _index "World")) _autosmoothtangent) "Local" true)
      (Spline|SetTangentatSplinePoint _tempspline 1 (Math|Vector|SelectVector (* (Math|Vector|Normalize (.Tangent (Spline|GetLocationandTangentatSplinePoint _localsplinesource (+ _index 1) "World"))) (* (Math|Vector|VectorLength (- (Spline|GetLocationandTangentatSplinePoint _localsplinesource (+ _index 1) "World") (Spline|GetLocationandTangentatSplinePoint _localsplinesource _index "World"))) (Variables|SplineOptions|GetTension))) (.Tangent (Spline|GetLocationandTangentatSplinePoint _localsplinesource (+ _index 1) "World")) _autosmoothtangent) "Local" true)
      (Landscape|Editor|EditorApplySpline _array_element _tempspline (* Width (.y (Spline|GetScaleatSplinePoint (LocalVariables|GetLocalSplineSource) _index))) (* Width (.y (Spline|GetScaleatSplinePoint (LocalVariables|GetLocalSplineSource) (+ _index 1)))) Falloff Falloff (Spline|GetRollatSplinePoint (LocalVariables|GetLocalSplineSource) _index "World") (Spline|GetRollatSplinePoint (LocalVariables|GetLocalSplineSource) (+ _index 1) "World") 80 true true 0 ""))
    (Components|DestroyComponent (Variables|Default|GetTempSpline))))

(fn ConfrormSplineToLanscape
  (params (Spline :type Spline%20Component%20Object%20Reference))
  (for _index (range (Spline|GetNumberOfSplinePoints Spline))
    (bind _outhit (Collision|LineTraceByChannel :Start (+ (Spline|GetLocationatSplinePoint Spline _index "World") "0, 0,20000.000000") :End (+ (Math|Vector|MakeVector 0.0 0.0 -515165159424.0) (Spline|GetLocationatSplinePoint Spline _index "World")) :TraceChannel "TraceTypeQuery10" :bTraceComplex false :DrawDebugType "ForDuration" :bIgnoreSelf true :TraceColor "(R=1.000000,G=0.000000,B=0.000000,A=1.000000)" :TraceHitColor "(R=0.000000,G=1.000000,B=0.000000,A=1.000000)" :DrawTime 5.0))
    (Spline|SetLocationatSplinePoint Spline _index (+ (.Location (Collision|BreakHitResult _outhit)) (Math|Vector|MakeVector 0.0 0.0 (* (Variables|Default|GetConfromPushUpZ) -1.0))) "World" true)))

(fn DamageInRadius
  (params (DamagePoint :type Vector) (Radius :type Float%20%28double-precision%29))
  (bind (_outactors _returnvalue) (Collision|SphereOverlapActors DamagePoint Radius (Utilities|Array|MakeArray "ObjectTypeQuery1" "ObjectTypeQuery5") 0 (Variables|Default|GetActorsToDestroy)))
  (if _returnvalue
    (for _array_element _outactors
      (bind _asbuilding_actor (Utilities|Casting|CastToBuildingActor _array_element))
      (Utilities|Casting|CastToCar_Copper _asbuilding_actor
        (:CastFailed
          (Utilities|Array|AddUnique (Variables|Default|GetActorsToDestroy) _asbuilding_actor))))
    (CallFunction|SetDestructionTimers)))

(fn DebugUpdateMaterialParameters
  (for _array_element (Variables|Default|GetMaterialList)
    (Rendering|Material|SetScalarParameterValue _array_element "ProgressOffset" (Variables|SplineOptions|GetProgressOffset) :!class /Script/Engine.MaterialInstanceDynamic)
    (for _array_element_2 (Variables|Default|GetMaterialList_SplineRight)
      (Rendering|Material|SetScalarParameterValue _array_element_2 "ProgressOffset" (Variables|SplineOptions|GetProgressOffset) :!class /Script/Engine.MaterialInstanceDynamic)
      (for _array_element_3 (Variables|Default|GetMaterialList_SplineLeft)
        (Rendering|Material|SetScalarParameterValue _array_element_3 "ProgressOffset" (Variables|SplineOptions|GetProgressOffset) :!class /Script/Engine.MaterialInstanceDynamic)))))

(fn FindClosestAndFurthestPoints
  (bind _epicenter (Variables|Default|GetEpicenter))
  (LocalVariables|SetClosestPoint (Math|Vector|VectorLength (- (Variables|Default|GetEpicenter) (Spline|GetLocationatTime (Variables|Default|GetSpline) 0.0 "World" true))))
  (for _array_element (Utilities|Array|MakeArray (Variables|Default|GetSpline) (Variables|Default|GetSplineLeft) (Variables|Default|GetSplineRight))
    (Utilities|IsValid _array_element
      (:Is%20Valid
        (if (> (Math|Vector|VectorLength (- _epicenter (Spline|GetLocationatTime _array_element 1.0 "World" true))) (LocalVariables|GetFurthestPoint))
          (LocalVariables|SetFurthestPoint (Math|Vector|VectorLength (- _epicenter (Spline|GetLocationatTime _array_element 1.0 "World" true))))))))
  (Variables|SplineOptions|SetProgressOffset (Math|Float|MapRangeClamped (LocalVariables|GetClosestPoint) 0.0 (Variables|Default|GetMaxDistance) 0.0 1e+01)))

(fn MoveDamage
  (bind _returnvalue_location (Math|Transform|BreakTransform (Spline|GetTransformatTime (Variables|Default|GetSplineRight) (Variables|Default|GetCrackOffsetProgressYSplit) "World" true false)))
  (bind _returnvalue_location_2 (Math|Transform|BreakTransform (Spline|GetTransformatTime (Variables|Default|GetSplineLeft) (Variables|Default|GetCrackOffsetProgressYSplit) "World" true false)))
  (bind _returnvalue_location_3 (Math|Transform|BreakTransform (Spline|GetTransformatTime (Variables|Default|GetSpline) (LocalVariables|GetCrackOffsetProgress) "World" true false)))
  (bind _returnvalue (Rendering|Material|GetScalarParameterValue "/Game/Athena/Environments/Landscape/MPC/MPC_Landscape.MPC_Landscape" "CrackProgression" :!class /Script/Engine.KismetMaterialLibrary))
  (Variables|Default|SetCurrentGlobalCrackProgression _returnvalue)
  (LocalVariables|SetCrackOffsetProgress (/ (- (Variables|Default|GetCurrentGlobalCrackProgression) (Variables|SplineOptions|GetProgressOffset)) (Math|Conversions|ToFloat(Integer) (Variables|Default|GetMaterialListLength))))
  (if (and (>= (LocalVariables|GetCrackOffsetProgress) 0.1) (<= (LocalVariables|GetCrackOffsetProgress) 0.95))
    (CallFunction|DamageinRadius _returnvalue_location_3 (Variables|Default|GetRadius))
    (elif (Variables|SplineOptions|GetYSplit)
      (Variables|Default|SetCrackOffsetProgressYSplit (/ (- (Variables|Default|GetCurrentGlobalCrackProgression) (+ (Variables|SplineOptions|GetProgressOffset) (Math|Conversions|ToFloat(Integer) (Variables|Default|GetMaterialListLength)))) (Math|Conversions|ToFloat(Integer) (Utilities|Array|Length (Variables|Default|GetMaterialList_SplineLeft)))))
      (if (and (>= (Variables|Default|GetCrackOffsetProgressYSplit) 0.1) (<= (Variables|Default|GetCrackOffsetProgressYSplit) 0.95))
        (Utilities|FlowControl|Sequence
          (:then_0
            (CallFunction|DamageinRadius _returnvalue_location_2 (Variables|Default|GetRadius)))
          (:then_1
            (CallFunction|DamageinRadius _returnvalue_location (Variables|Default|GetRadius))))))))

(fn MoveParticleEffect)

(fn SetMainTravelAudioEnabled
  (params (Enabled :type Boolean)))

(fn SetMaterialParametersOnList
  (params (NewParam :type Array%20of%20Material%20Instance%20Dynamic%20Object%20References) (IsStart :type Boolean) (isEnd :type Boolean) (OffsetProgression :type Integer))
  (for (_array_element _array_index) NewParam
    (Utilities|FlowControl|Sequence
      (:then_0
        (if (and (== _array_index 0) IsStart)
          (Rendering|Material|SetScalarParameterValue _array_element "First" 1.0 :!class /Script/Engine.MaterialInstanceDynamic)
          (else
            (Rendering|Material|SetScalarParameterValue _array_element "First" 0.0 :!class /Script/Engine.MaterialInstanceDynamic))))
      (:then_1
        (if (and (== _array_index (Utilities|Array|LastIndex NewParam)) isEnd)
          (Rendering|Material|SetScalarParameterValue _array_element "Last" 1.0 :!class /Script/Engine.MaterialInstanceDynamic)
          (else
            (Rendering|Material|SetScalarParameterValue _array_element "Last" 0.0 :!class /Script/Engine.MaterialInstanceDynamic))))
      (:then_2
        (Rendering|Material|SetScalarParameterValue _array_element "ProgressOffset" (+ (Math|Conversions|ToFloat(Integer) (+ _array_index OffsetProgression)) (Variables|SplineOptions|GetProgressOffset)) :!class /Script/Engine.MaterialInstanceDynamic)))))

(fn SetYSplitTravelAudioEnabled
  (params (Enabled :type Boolean)))

(fn TransformEffect
  (params (Effect :type Scene%20Component%20Object%20Reference) (Spline :type Spline%20Component%20Object%20Reference) (SplineTime :type Float%20%28double-precision%29))
  (bind (_returnvalue_location _returnvalue_rotation _returnvalue_scale) (Math|Transform|BreakTransform (Spline|GetTransformatTime Spline SplineTime "World" false false)))
  (Transformation|SetWorldTransform Effect (Math|Transform|MakeTransform _returnvalue_location _returnvalue_rotation (Math|Conversions|ToVector(Float) (Math|Float|Clamp(Float) (.y _returnvalue_scale) 1.0 2.5))) false false))

(fn ConstructionScript
  (bind _materiallist (Variables|Default|GetMaterialList))
  (bind _returnvalue (Utilities|Array|Length _materiallist))
  (bind _materiallist_2 (Variables|Default|GetMaterialList))
  (bind _returnvalue_2 (Utilities|Array|Length _materiallist_2))
  (bind (_location _tangent) (Spline|GetLocationandTangentatSplinePoint (Variables|Default|GetSpline) 1 "Local"))
  (bind (_location_2 _tangent_2) (Spline|GetLocationandTangentatSplinePoint (Variables|Default|GetSpline) 0 "Local"))
  (bind _splinemeshcomponent (Variables|SplineMeshActor|GetSplineMeshComponent))
  (bind _zigzagstrenght (Variables|SplineOptions|GetZigZagStrenght))
  (bind _spline (Variables|Default|GetSpline))
  (bind _spline_2 (Variables|Default|GetSpline))
  (bind _splinemeshcomponent_2 (Variables|SplineMeshActor|GetSplineMeshComponent))
  (bind _splinemeshcomponent_3 (Variables|SplineMeshActor|GetSplineMeshComponent))
  (bind _splinemeshcomponent_4 (Variables|SplineMeshActor|GetSplineMeshComponent))
  (Utilities|FlowControl|Sequence
    (:then_0
      (Utilities|Array|Clear (Variables|Default|GetSplineMeshes))
      ;[1 1 1] Set First Splien Mesh Component
      (SplineMesh|SetStartandEnd _splinemeshcomponent _location_2 (Math|Vector|SelectVector (* (Math|Vector|Normalize _tangent_2) (* (Math|Vector|VectorLength (- _location_2 _location)) (Variables|SplineOptions|GetTension))) _tangent_2 (Variables|SplineOptions|GetAutoSmoothTangent)) _location (Math|Vector|SelectVector (* (Math|Vector|Normalize _tangent) (* (Math|Vector|VectorLength (- _location_2 _location)) (Variables|SplineOptions|GetTension))) _tangent (Variables|SplineOptions|GetAutoSmoothTangent)) true :!class /Script/Engine.SplineMeshComponent)
      (bind _returnvalue_3 (Rendering|Material|CreateDynamicMaterialInstance _splinemeshcomponent 0 (Variables|SplineOptions|GetMaterialOverride) "None" :!class /Script/Engine.PrimitiveComponent))
      (Utilities|Array|Add (Variables|Default|GetMaterialList) _returnvalue_3)
      (SplineMesh|SetEndRollRadians _splinemeshcomponent_3 (Math|Trig|DegreesToRadians (Spline|GetRollatSplinePoint _spline_2 1 "Local")) true)
      (SplineMesh|SetStartRollRadians _splinemeshcomponent_3 (Math|Trig|DegreesToRadians (Spline|GetRollatSplinePoint _spline_2 0 "Local")) true)
      (SplineMesh|SetStartScale _splinemeshcomponent_2 (Math|Vector2D|MakeVector2D (.y (Spline|GetScaleatSplinePoint _spline 0)) (.z (Spline|GetScaleatSplinePoint _spline 0))) true :!class /Script/Engine.SplineMeshComponent)
      (SplineMesh|SetEndScale _splinemeshcomponent_2 (Math|Vector2D|MakeVector2D (.y (Spline|GetScaleatSplinePoint _spline 1)) (.z (Spline|GetScaleatSplinePoint _spline 1))) true :!class /Script/Engine.SplineMeshComponent)
      (SplineMesh|SetEndOffset _splinemeshcomponent_4 (Math|Vector2D|MakeVector2D (Math|Float|SelectFloat 0.0 (* _zigzagstrenght -1.0) (and (== (Spline|GetNumberOfSplinePoints (Variables|Default|GetSpline)) 2) (Variables|SplineOptions|GetYSplit)))) true)
      (SplineMesh|SetStartOffset _splinemeshcomponent_4 (Math|Vector2D|MakeVector2D _zigzagstrenght) true)
      ;[1 1 1] Add Splinemesh Compontent for first Spline
      (CallFunction|AddSplineComponentsToSpline (Variables|Default|GetSpline) 1 false (Variables|SplineOptions|GetYSplit) (Variables|Default|GetMaterialList)))
    (:then_1
      ;[1 1 1] Add YSplit Conector mesh
      (if (Variables|SplineOptions|GetYSplit)
        (bind _returnvalue_4 (AddComponent|NODE_AddStaticMeshComponent-0 :RelativeTransform (Math|Transform|MakeTransform (Spline|GetLocationatSplinePoint (Variables|Default|GetSpline) (Spline|GetNumberOfSplinePoints (Variables|Default|GetSpline)) "Local") (Spline|GetRotationatSplinePoint (Variables|Default|GetSpline) (Spline|GetNumberOfSplinePoints (Variables|Default|GetSpline)) "Local") (Math|Vector|MakeVector (.y (Spline|GetScaleatSplinePoint (Variables|Default|GetSpline) (Spline|GetNumberOfSplinePoints (Variables|Default|GetSpline)))) (.y (Spline|GetScaleatSplinePoint (Variables|Default|GetSpline) (Spline|GetNumberOfSplinePoints (Variables|Default|GetSpline)))) (.z (Spline|GetScaleatSplinePoint (Variables|Default|GetSpline) (Spline|GetNumberOfSplinePoints (Variables|Default|GetSpline))))))))
        (bind _output_get (Variables|Default|SetConnectorMesh _returnvalue_4))
        (bind _returnvalue_5 (Rendering|Material|CreateDynamicMaterialInstance _output_get 0 (Variables|SplineOptions|GetMaterialOverride) "None" :!class /Script/Engine.PrimitiveComponent))
        (Utilities|Array|Add (Variables|Default|GetMaterialList) _returnvalue_5)
        (Transformation|SetRelativeTransform (Variables|Default|GetSplineLeft) (Math|Transform|MakeTransform (.location (Transformation|GetSocketTransform _returnvalue_4 "Left" "RTS_Actor")) (.rotation (Transformation|GetSocketTransform _returnvalue_4 "Left" "RTS_Actor"))) false false)
        (Transformation|SetRelativeTransform (Variables|Default|GetSplineRight) (Math|Transform|MakeTransform (.location (Transformation|GetSocketTransform _returnvalue_4 "Right" "RTS_Actor")) (.rotation (Transformation|GetSocketTransform _returnvalue_4 "Right" "RTS_Actor"))) false false)
        (else
          (Components|DestroyComponent (Variables|Default|GetSplineLeft)))))
    (:then_2
      ;[1 1 1] Ad YSplit SplineMesh Components
      (if (Variables|SplineOptions|GetYSplit)
        (CallFunction|AddSplineComponentsToSpline (Variables|Default|GetSplineLeft) 0 true false (Variables|Default|GetMaterialList_SplineLeft))
        (CallFunction|AddSplineComponentsToSpline (Variables|Default|GetSplineRight) 0 true false (Variables|Default|GetMaterialList_SplineRight))))
    (:then_3
      ;[1 1 1] Set material
      (CallFunction|SetMaterialParametersonList (Variables|Default|GetMaterialList) true (not (Variables|SplineOptions|GetYSplit)))
      (if (Variables|SplineOptions|GetYSplit)
        (CallFunction|SetMaterialParametersonList (Variables|Default|GetMaterialList_SplineLeft) false true _returnvalue_2)
        (CallFunction|SetMaterialParametersonList (Variables|Default|GetMaterialList_SplineRight) false true _returnvalue_2)))
    (:then_4
      (if (> _returnvalue 1)
        (Variables|Default|SetMaterialListLength _returnvalue)
        (else
          (Variables|Default|SetMaterialListLength 1))))))
